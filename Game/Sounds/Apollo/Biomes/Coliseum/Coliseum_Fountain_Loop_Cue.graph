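# Sound cue for the looping fountain ambience in the Apollo Coliseum biome, in the Athena ambient sound class.
# A single looping wave player feeds a modulator that adds a slight upward pitch variation at constant volume, with the cue's overall volume reduced.
# It references a fountain focus-cone attenuation asset, also carries capsule-shaped attenuation override values, and enables a concurrency override.
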SoundCue /Game/Sounds/Apollo/Biomes/Coliseum/Coliseum_Fountain_Loop_Cue
root=Modulator_1
nodes [
 {id=Modulator_1, children [WavePlayer_0], properties {pitchMax=1.1, pitchMin=1, volumeMax=1, volumeMin=1}}
 {id=WavePlayer_0, properties {bLooping=true, soundWaveAssetPtr=&/Coliseum_Fountain_Loop_01}}
]
properties {
 attenuationSettings=/Game/Sounds/Attenuation/Fort_Coliseum_Fountain_Focus_Cone_Att
 bOverrideConcurrency=true
 maxDistance=1250
 soundClassObject=&Fort_Ambient_Athena
 volumeMultiplier=0.312
 attenuationOverrides {attenuationShape=Capsule, falloffDistance=950, stereoSpread=0, attenuationShapeExtents=<250 90>}
}
